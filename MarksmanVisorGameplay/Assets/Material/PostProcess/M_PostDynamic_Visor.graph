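# Post-process material for the Marksman Visor that draws a glowing scan pulse expanding through the world over the scene.
# A parameter collection supplies origin, radius, timestamp and duration to drive distance-based wave and gradient masks, with a static switch for looping preview.
# Colour layers combine a world-aligned triangle grid, screen-space scanlines and scene-depth edge detection.
# The glow is divided by eye adaptation before compositing, and low-end/mobile paths use simplified shading without edge detection.
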Material /MarksmanVisorGameplay/Assets/Material/PostProcess/M_PostDynamic_Visor
graph {
 properties {bAutomaticallySetUsageInEditor=false, materialDomain=MD_PostProcess, usedShadingModels=Default Lit}
 nodes [
  n1 = ScalarParameter() {defaultValue=1024, group=TechArt, parameterName=WorldTextureScale}
  Divide(A=NamedRerouteUsage(), B=@n1)
  n2 = ComponentMask(None=NamedRerouteUsage()) {r=true}
  n3 = VectorParameter() {group=ShadingColor, parameterName=GradientColor01, defaultValue=<rgba 0 1 0.1503 0.5>}
  n4 = Multiply(A=@n3.RGB, B=@n3.RGB)
  n5 = VectorParameter() {group=ShadingColor, parameterName=GradientColor02, defaultValue=<rgba 1 0.7001 0 0.5>}
  n6 = Multiply(A=@n5.RGB, B=@n5.RGB)
  n7 = NamedRerouteUsage()
  n8 = VectorParameter() {group=ShadingColor, parameterName=PatternColor01, defaultValue=<rgba 0 1 0.178 0.05>}
  n9 = VectorParameter() {group=ShadingColor, parameterName=PatternColor02, defaultValue=<rgba 1 0.713 0 0.05>}
  n10 = LinearInterpolate(A=Multiply(A=@n8.RGB, B=@n8.RGB), B=Multiply(A=@n9.RGB, B=@n9.RGB), Alpha=@n7)
  n11 = Add(A=Multiply(A=OneMinus(None=@n2), B=LinearInterpolate(A=@n4, B=@n6, Alpha=@n7)), B=Multiply(A=NamedRerouteUsage(), B=@n10))
  n12 = Multiply(A=NamedRerouteUsage(), B=NamedRerouteUsage())
  n13 = NamedRerouteUsage()
  n14 = ComponentMask(None=NamedRerouteUsage()) {r=true}
  n15 = Multiply(A=NamedRerouteUsage(), B=LinearInterpolate(Alpha=OneMinus(None=@n14)) {constA=0.2}) {constB=0.2}
  n16 = NamedRerouteUsage()
  n17 = NamedRerouteUsage()
  n18 = VectorParameter() {group=ShadingColor, parameterName=EdgeDetectColor01, defaultValue=<rgba 0 0.1667 0.0035 1>}
  n19 = Multiply(A=@n18.RGB, B=@n18.RGB)
  n20 = VectorParameter() {group=ShadingColor, parameterName=EdgeDetectColor02, defaultValue=<rgba 1 0.713 0 0.05>}
  n21 = Multiply(A=@n20.RGB, B=@n20.RGB)
  n22 = NamedRerouteUsage()
  n23 = Multiply(A=Power(Base=NamedRerouteUsage()), B=Multiply(A=Multiply(A=@n16, B=@n17), B=LinearInterpolate(A=@n19, B=@n21, Alpha=@n22)))
  n24 = Add(A=Add(A=Multiply(A=@n11, B=@n12), B=Multiply(A=@n13, B=@n15)), B=@n23)
  MaterialFunctionCall!(Default=@n24, Low/Mobile=Add(A=Multiply(A=@n12, B=@n10), B=@n23)) {materialFunction=/Game/Effects/Fort_Effects/Materials/_Functions/MF_LowMobile_PathingSwitch}
  n25 = NamedRerouteUsage()
  n26 = Multiply(A=Time()) {constB=0.5}
  n27 = CollectionParameter() {collection=/Game/Weapons/Prototype/FringePlank/PostProcess/MPC_FringePlank_Interrogate, parameterName=Radius}
  n28 = NamedRerouteDeclaration(None=MaterialFunctionCall(
    TimestampA=CollectionParameter() {collection=&MPC_FringePlank_Interrogate, parameterName=ActivatedTimestamp}
    TransitionLength_In=CollectionParameter() {collection=&MPC_FringePlank_Interrogate, parameterName=Duration}
    Transition In Only?=StaticBool() {value=true}
   ) {materialFunction=&MF_Timestamp_InOut_01}.Linear) {name=ActiveAlpha, nodeColor=<rgba 0.4297 1 0 1>}
  n29 = Multiply(A=ScalarParameter() {defaultValue=2, parameterName=InteralSpeedUp}, B=@n28)
  n30 = StaticSwitchParameter(
   True=Multiply(A=ScalarParameter() {defaultValue=20000, group=0_Vis, parameterName=ScanningRadius}, B=Frac(None=@n26))
   False=Multiply(A=NamedRerouteDeclaration(None=@n27) {name=RadiusTotal, nodeColor=<rgba 0 0.3672 1 1>}, B=Saturate(None=@n29))
  ) {group=0_Vis, parameterName=VisScanning}
  n31 = Subtract(A=@n25, B=@n30)
  n32 = Saturate(None=Divide(A=@n31, B=ScalarParameter() {defaultValue=400, parameterName=ScanlineWidth}))
  n33 = OneMinus(None=@n32)
  n34 = ScalarParameter() {defaultValue=5000, parameterName=GradientWidth}
  n35 = Subtract(A=@n31, B=Saturate(None=Subtract(A=@n30, B=@n34)))
  n36 = OneMinus(None=Saturate(None=Divide(A=@n35, B=@n34)))
  Multiply(A=NamedRerouteDeclaration(None=@n33) {name=MaxArea, nodeColor=<rgba 1 0.4219 0 1>}, B=@n36)
  n37 = TextureSample(
   UVs=Multiply(A=ScreenPosition().ViewportUV, B=ScalarParameter() {defaultValue=5, group=TechArt, parameterName=ScreenTilling})
   Tex=NamedRerouteUsage()
  ) {samplerSource=SSM_Wrap_WorldGroupSettings, samplerType=SAMPLERTYPE_Masks}
  n38 = VectorParameter() {group=ShadingColor, parameterName=ScreenColor01, defaultValue=<rgba 0 1 0.0535 1>}
  n39 = Multiply(A=@n38.RGB, B=@n38.RGB)
  NamedRerouteDeclaration(None=Multiply(A=NamedRerouteDeclaration(None=@n37.B) {name=ScreenLinesAlpha, nodeColor=<rgba 0.4062 1 0 1>}, B=@n39)) {name=ScreenLines, nodeColor=<rgba 1 0 0.375 1>}
  n40 = WorldPosition()
  n41 = StaticSwitchParameter(
   True=VectorParameter() {group=TechArt, parameterName=CenterScanning, defaultValue=<rgba 0 0 0 1>}.RGB
   False=ComponentMask(None=CollectionParameter() {collection=&MPC_FringePlank_Interrogate, parameterName=OriginPositionWS}) {b=true, g=true, r=true}
  ) {group=0_Vis, parameterName=VisScanning}
  NamedRerouteDeclaration(None=Subtract(A=@n40.XYZ, B=@n41)) {name=OffsetWorldPos, nodeColor=<rgba 0 1 0.4609 1>}
  n42 = Distance(A=@n40.XYZ, B=@n41)
  n43 = NamedRerouteDeclaration(None=@n42) {name=DistanceWS, nodeColor=<rgba 1 0 0.3516 1>}
  n44 = ScalarParameter() {defaultValue=100, group=TechArt, parameterName=ScanningWidth}
  n45 = Divide(A=Abs(None=Subtract(A=@n43, B=@n30)), B=@n44)
  NamedRerouteDeclaration(None=OneMinus(None=Saturate(None=@n45))) {name=WaveArea, nodeColor=<rgba 0 0.9062 1 1>}
  n46 = Divide(A=@n42, B=ScalarParameter() {defaultValue=6000, parameterName=CenterGradient})
  NamedRerouteDeclaration(None=OneMinus(None=Saturate(None=@n46))) {name=CenterGradient, nodeColor=<rgba 1 0.2344 0 1>}
  NamedRerouteDeclaration(None=@n39) {name=ScreenColor, nodeColor=<rgba 0 1 0.5078 1>}
  n47 = VertexNormalWS()
  n48 = SceneTexture() {sceneTextureId=PPI_WorldNormal}
  n49 = MaterialFunctionCall(
   TextureObject=TextureObjectParameter() {group=TechArt, parameterName=WordGrid, texture=/MarksmanVisorGameplay/Assets/Texture/T_VisorTriGrid}
   TextureSize=@n1
   WorldPosition=NamedRerouteUsage()
   World Space Normal=NamedRerouteDeclaration(None=ShadingPathSwitch(Default=@n47, Deferred=@n48.Color, Forward=@n47, Mobile=@n47)) {name=WorldN, nodeColor=<rgba 1 0.6094 0 1>}
  ) {materialFunction=&WorldAlignedTexture}
  NamedRerouteDeclaration(None=@n49.XYZ Texture) {name=WorldGridOG, nodeColor=<rgba 1 0.4219 0 1>}
  n50 = ScreenPosition()
  n51 = Multiply(B=SceneTexelSize()) {constA=3}
  n52 = Add(A=@n50.ViewportUV, B=Multiply(A=@n51, B=Constant2Vector() {g=-1, r=-1}))
  n53 = SceneTexture() {sceneTextureId=PPI_SceneDepth}
  n54 = Add(A=@n50.ViewportUV, B=Multiply(A=@n51, B=Constant2Vector() {g=-1, r=1}))
  n55 = Add(
   A=Subtract(A=SceneTexture(UVs=@n52) {sceneTextureId=PPI_SceneDepth}.Color, B=@n53.Color)
   B=Subtract(A=SceneTexture(UVs=@n54) {sceneTextureId=PPI_SceneDepth}.Color, B=@n53.Color)
  )
  n56 = Multiply(A=@n51, B=Constant2Vector() {g=1, r=1})
  n57 = Subtract(A=SceneTexture(UVs=Add(A=@n50.ViewportUV, B=@n56)) {sceneTextureId=PPI_SceneDepth}.Color, B=@n53.Color)
  n58 = Constant2Vector() {g=1, r=-1}
  n59 = SceneTexture(UVs=Add(A=@n50.ViewportUV, B=Multiply(A=@n51, B=@n58))) {sceneTextureId=PPI_SceneDepth}
  n60 = Abs(None=Add(A=Add(A=@n55, B=@n57), B=Subtract(A=@n59.Color, B=@n53.Color)))
  n61 = Saturate(None=Multiply(A=ComponentMask(None=@n60) {r=true}, B=ScalarParameter() {defaultValue=0.01, parameterName=EdgeDetectionScaleA}))
  n62 = SmoothStep(Value=@n61) {constMin=0.3}
  n63 = ScalarParameter() {defaultValue=10, parameterName=EdgeDetectionScaleB}
  NamedRerouteDeclaration(None=MaterialFunctionCall(Default=Multiply(A=@n62, B=@n63), Low/Mobile=Constant() {r=1}) {materialFunction=&MF_LowMobile_PathingSwitch}.Result) {name=EdgeDetection, nodeColor=<rgba 0.8516 0 1 1>}
  NamedRerouteDeclaration(None=TextureObject() {samplerType=SAMPLERTYPE_Masks, texture=/MarksmanVisorGameplay/Assets/Texture/T_VisorScanning}) {name=ScanningEffect, nodeColor=<rgba 1 0.9844 0 1>}
  NamedRerouteDeclaration(None=Multiply(A=@n32, B=@n33)) {name=DebugWave, nodeColor=<rgba 1 0 0.5156 1>}
  n64 = Divide(A=@n25, B=ScalarParameter() {defaultValue=100, group=TechArt, parameterName=GlobalGradient})
  NamedRerouteDeclaration(None=Multiply(A=@n36, B=Saturate(None=@n64))) {name=GradientArea, nodeColor=<rgba 1 0 0.3984 1>}
  NamedRerouteDeclaration(None=SmoothStep(Value=@n28) {constMax=0.7, constMin=0.4}) {name=FadeOUTAlpha, nodeColor=<rgba 1 0 0.375 1>}
  n65 = Add(A=@n44, B=@n30)
  n66 = Divide(A=@n42, B=@n65)
  NamedRerouteDeclaration(None=OneMinus(None=Saturate(None=@n66))) {name=ScannedGradient, nodeColor=<rgba 0 1 0.25 1>}
  n67 = SmoothStep(Min=Subtract(A=@n65) {constB=1000}, Max=@n65, Value=@n42)
  NamedRerouteDeclaration(None=OneMinus(None=Saturate(None=@n67))) {name=ScannedArea, nodeColor=<rgba 0 0.8828 1 1>}
  NamedRerouteDeclaration(None=@n1) {name=TextureScale, nodeColor=<rgba 0 1 0.0625 1>}
  n68 = ComponentMask(None=@n49.XYZ Texture) {g=true}
  n69 = Multiply(A=Time(), B=ScalarParameter() {defaultValue=1, group=TechArt, parameterName=Speed02})
  NamedRerouteDeclaration(None=Frac(None=Add(A=@n68, B=@n69))) {name=WorldGrid_Pattern, nodeColor=<rgba 1 0 0.2578 1>}
  n70 = Saturate(None=Divide(A=@n42, B=Constant() {r=2048}))
  NamedRerouteDeclaration(None=SmoothStep(Value=OneMinus(None=@n70)) {constMax=0.8}) {name=CenterYellow, nodeColor=<rgba 0 0.2266 1 1>}
  Round(None=@n61)
  n71 = Multiply(A=OneMinus(None=NamedRerouteUsage()), B=@n24)
  n72 = EyeAdaptation()
  n73 = ComponentMask(None=SceneTexture() {sceneTextureId=PPI_PostProcessInput0}.Color) {b=true, g=true, r=true}
  Saturate(None=Add(A=Divide(A=@n71, B=@n72.EyeAdaptation), B=NamedRerouteDeclaration(None=@n73) {name=ColorIN, nodeColor=<rgba 1 0 0.3516 1>}))
 ]
}
